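# Physics asset for the Llama chain charm accessory, giving its hanging chain simulated ragdoll-style motion.
# Four box bodies cover the root and three chain bones, with the much larger box on the last bone, chain_03, likely enclosing the llama pendant.
# Constraints link each bone to its parent in a single chain, with moderate swing limits and tight twist limits so it dangles without spinning.
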PhysicsAsset /Game/Accessories/Charms/Charm_Llama_Chain/Meshes/Charm_Llama_Chain_Physics
bodies {
 chain_01 [{shapeType=Box, extentX=2.51, extentY=2.51, extentZ=1.01, center=<0.9022 0 -0.0102>}]
 chain_02 [{shapeType=Box, extentX=2.01, extentY=2.01, extentZ=1.01, center=<0.9572 0 -0.0102>}]
 chain_03 [{shapeType=Box, extentX=13.1272, extentY=8.2132, extentZ=5.807, center=<4.5964 0.2088 -0.0756>}]
 root [{shapeType=Box, extentX=1.5648, extentY=3.6025, extentZ=3.8316, center=<-0.7923 0 -0.4221>}]
}
constraints [
 {bone1Name=chain_01, bone2Name=root, swing1LimitDegrees=45, swing2LimitDegrees=45, twistLimitDegrees=5}
 {bone1Name=chain_02, bone2Name=chain_01, swing1LimitDegrees=45, swing2LimitDegrees=45, twistLimitDegrees=5}
 {bone1Name=chain_03, bone2Name=chain_02, swing1LimitDegrees=45, swing2LimitDegrees=45, twistLimitDegrees=5}
]
